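# Backpack cosmetic screen material that displays a two-digit kill counter driven by Kills-Tens and Kills-Ones.
# Parameters shift number-sheet UVs, masked into digit slots on UV1, with UseSingleDigit blending to a faint "8" tens placeholder.
# UseNerdBackpack selects dark digits over diffuse with a FlashControl-driven backlight, or glowing noise-distorted digits, beams and vertex-color-masked normal offset.
# Specular masks, normal and emissive textures complete the attributes, passed through MF_TechArtOverrides_accessories.
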Material /BRCosmetics/Accessories/FORT_Backpacks/M_Med_Nerd_Backpack/Materials/M_M_Med_Nerd_Backpack_Screen
graph {
 properties {
  bUseMaterialAttributes=true
  bUsedWithMeshDeformer=true
  bUsedWithSkeletalMesh=true
  pixelDepthOffsetMode=PDOM_Legacy
  usedShadingModels=Default Lit
 }
 nodes [
  Constant() {r=0.25}
  Constant() {r=-1}
  Constant() {r=0.01}
  Constant() {r=0.6}
  Constant() {r=8}
  n1 = TextureSampleParameter2D() {
   parameterName=DiffuseMask
   samplerSource=SSM_Wrap_WorldGroupSettings
   samplerType=SAMPLERTYPE_Masks
   texture=/Engine/EngineMaterials/DefaultDiffuse_TC_Masks
  }
  n2 = Constant() {r=4}
  n3 = TextureSampleParameter2D() {parameterName=DiffuseTexture, samplerSource=SSM_Wrap_WorldGroupSettings, texture=&T_BlacksRGB}
  n4 = VectorParameter() {parameterName=DiffuseTextureMod, defaultValue=<rgba 2.75 2.6487 2.2 1>}
  n5 = Constant() {r=15}
  n6 = Constant() {r=44.3}
  n7 = Add(
   A=TextureCoordinate() {coordinateIndex=1}
   B=MaterialFunctionCall(X=ScalarParameter() {defaultValue=0.01, parameterName=H-Offset-Base}, Y=Constant()) {materialFunction=&MakeFloat2}.Result
  )
  n8 = Multiply(A=ComponentMask(None=@n7) {r=true}, B=Constant() {r=-100})
  n9 = Add(A=Constant() {r=-4.5}, B=Abs(None=Add(A=@n6, B=@n8)))
  n10 = OneMinus(None=Saturate(None=Multiply(A=@n5, B=@n9)))
  n11 = Constant() {r=0.0938}
  n12 = ScalarParameter() {defaultValue=-0.365, parameterName=H-Offset-TensDigit}
  n13 = Add(A=Multiply(A=ScalarParameter() {parameterName=Kills-Tens}, B=@n11), B=@n12)
  n14 = Constant()
  n15 = TextureSampleParameter2D(UVs=Add(A=MaterialFunctionCall(X=@n13, Y=@n14) {materialFunction=&MakeFloat2}.Result, B=@n7)) {
   desc=Tenths
   parameterName=NumberSheet
   samplerSource=SSM_Wrap_WorldGroupSettings
   samplerType=SAMPLERTYPE_LinearColor
   texture=/BRCosmetics/Accessories/FORT_Backpacks/M_Med_Nerd_Backpack/Textures/T_NumberSet_Large
  }
  n16 = Add(A=@n8, B=Constant() {r=53.85})
  n17 = Multiply(A=Add(A=Abs(None=@n16), B=Constant() {r=-4.5}), B=Constant() {r=15})
  n18 = Add(
   A=Multiply(A=ScalarParameter() {parameterName=Kills-Ones}, B=@n11)
   B=ScalarParameter() {defaultValue=-0.46, parameterName=H-Offset-OnesDigit}
  )
  n19 = Constant()
  n20 = TextureSampleParameter2D(UVs=Add(A=@n7, B=MaterialFunctionCall(X=@n18, Y=@n19) {materialFunction=&MakeFloat2}.Result)) {
   desc=Ones
   parameterName=NumberSheet
   samplerSource=SSM_Wrap_WorldGroupSettings
   samplerType=SAMPLERTYPE_LinearColor
   texture=&T_NumberSet_Large
  }
  n21 = Multiply(A=OneMinus(None=Saturate(None=@n17)), B=@n20.R)
  n22 = Add(A=Multiply(A=@n10, B=@n15.R), B=@n21)
  n23 = Constant() {r=0.15}
  n24 = Add(A=Multiply(A=Constant() {r=8}, B=@n11), B=@n12)
  n25 = Constant()
  n26 = TextureSampleParameter2D(UVs=Add(A=MaterialFunctionCall(X=@n24, Y=@n25) {materialFunction=&MakeFloat2}.Result, B=@n7)) {
   desc=Tenths
   parameterName=NumberSheet
   samplerSource=SSM_Wrap_WorldGroupSettings
   samplerType=SAMPLERTYPE_LinearColor
   texture=&T_NumberSet_Large
  }
  n27 = Multiply(A=@n26.R, B=@n10)
  n28 = LinearInterpolate(A=@n22, B=Add(A=Multiply(A=@n23, B=@n27), B=@n21), Alpha=ScalarParameter() {defaultValue=1, parameterName=UseSingleDigit})
  n29 = LinearInterpolate(
   A=Multiply(A=Power(Base=@n1.R, Exp=@n2), B=Multiply(A=@n3.RGB, B=@n4.RGB))
   B=VectorParameter() {parameterName=NumberColor, defaultValue=<rgba 0.03 0.03 0.03 1>}.RGB
   Alpha=@n28
  )
  n30 = Constant()
  n31 = TextureSampleParameter2D() {
   group=SpecularMasks
   parameterName=SpecularMasks_Texture
   samplerSource=SSM_Wrap_WorldGroupSettings
   texture=/Game/Characters/Fortress_Character/Textures/T_All_White
  }
  n32 = ScalarParameter() {group=SpecularMasks, parameterName=Metallic_MOD}
  n33 = ScalarParameter() {defaultValue=0.1, group=SpecularMasks, parameterName=Specular_MOD}
  n34 = ScalarParameter() {defaultValue=0.3, group=SpecularMasks, parameterName=Roughness_MOD}
  n35 = Multiply(A=TextureCoordinate() {coordinateIndex=1}, B=Constant2Vector() {g=1.5, r=1})
  n36 = Constant() {r=0.18}
  n37 = Constant()
  n38 = SphereMask(A=@n35, B=Constant2Vector() {g=0.2, r=0.325}, Radius=@n36, Hardness=@n37)
  n39 = SphereMask(A=@n35, B=Constant2Vector() {g=0.2, r=0.6}, Radius=@n36, Hardness=@n37)
  n40 = Multiply(A=OneMinus(None=@n28), B=Power(Base=Add(A=@n38, B=@n39), Exp=Constant() {r=2}))
  n41 = VectorParameter() {parameterName=EmissiveLightColor, defaultValue=<rgba 0.5 0.4381 0.1641 1>}
  n42 = Power(Base=@n22, Exp=ScalarParameter() {defaultValue=2, parameterName=Pre-Mask-Pow})
  n43 = Time()
  n44 = ScalarParameter() {defaultValue=0.6, parameterName=P-DistortSpeed}
  n45 = MaterialFunctionCall(
   X=ScalarParameter() {defaultValue=0.01, parameterName=P-DistortOffsetA}
   Y=ScalarParameter() {defaultValue=8, parameterName=P-DistortOffsetB}
  ) {materialFunction=&MakeFloat2}
  n46 = TextureCoordinate() {coordinateIndex=1}
  n47 = TextureSample(UVs=Add(A=Multiply(A=@n43, B=@n44), B=Multiply(A=@n45.Result, B=@n46))) {samplerType=SAMPLERTYPE_Grayscale, texture=&LowResBlurredNoise}
  n48 = LinearInterpolate(
   A=ScalarParameter() {defaultValue=0.25, parameterName=P-DistortLerpA}
   B=ScalarParameter() {defaultValue=-1, parameterName=P-DistortLerpB}
   Alpha=@n47.R
  )
  n49 = Add(A=TextureCoordinate(), B=Multiply(A=Saturate(None=@n48), B=ScalarParameter() {defaultValue=0.75, parameterName=P-DistortAmt}))
  n50 = ScalarParameter() {defaultValue=0.5, parameterName=P-SM-Rad}
  n51 = Constant()
  n52 = Add(
   A=SphereMask(A=@n49, B=Constant2Vector() {r=0.5}, Radius=@n50, Hardness=@n51)
   B=SphereMask(A=@n49, B=Constant2Vector() {g=1, r=0.5}, Radius=@n50, Hardness=@n51)
  )
  n53 = ScalarParameter() {defaultValue=2, parameterName=SM-Amt-Pow}
  n54 = LinearInterpolate(
   A=Multiply(A=ScalarParameter() {defaultValue=0.05, parameterName=SM-Amt-BG}, B=@n52)
   B=Multiply(A=ScalarParameter() {defaultValue=5, parameterName=SM-Amt}, B=Power(Base=@n52, Exp=@n53))
   Alpha=@n22
  )
  n55 = Multiply(A=@n22, B=ScalarParameter() {defaultValue=25, parameterName=NumberChangedBrightness})
  n56 = ScalarParameter() {defaultValue=0.2, parameterName=NumberChangedBrightnessAdd}
  n57 = ScalarParameter() {parameterName=FlashControl}
  n58 = Add(
   A=Multiply(A=Add(A=@n42, B=@n54), B=ScalarParameter() {defaultValue=2, parameterName=Pre-MaskBright2})
   B=Multiply(A=Add(A=@n55, B=@n56), B=@n57)
  )
  n59 = ComponentMask(None=TextureCoordinate()) {g=true}
  n60 = ScalarParameter() {defaultValue=0.5, parameterName=Beams}
  n61 = Multiply(A=Multiply(A=Multiply(A=@n59, B=@n60), B=@n47.R), B=ScalarParameter() {defaultValue=0.5, parameterName=BeamsAmt})
  n62 = VertexColor()
  n63 = StaticSwitchParameter(
   True=Multiply(A=ScalarParameter() {parameterName=FlashControl}, B=Multiply(A=@n40, B=@n41.RGB))
   False=Multiply(A=LinearInterpolate(A=@n58, B=@n61, Alpha=@n62.B), B=@n41.RGB)
  ) {defaultValue=true, parameterName=UseNerdBackpack}
  n64 = TextureSampleParameter2D() {parameterName=EmissiveTexture, samplerSource=SSM_Wrap_WorldGroupSettings, texture=&T_BlacksRGB}
  n65 = Add(
   A=Multiply(A=@n28, B=VectorParameter() {parameterName=NumberColor-Emissive, defaultValue=<rgba 0 0 0 1>}.RGB)
   B=Multiply(A=@n64.RGB, B=ScalarParameter() {defaultValue=1, parameterName=EmissiveTexture-Multiplier})
  )
  n66 = TextureSampleParameter2D() {
   parameterName=NormalTexture
   samplerSource=SSM_Wrap_WorldGroupSettings
   samplerType=SAMPLERTYPE_Normal
   texture=&FlatNormalBlue
  }
  n67 = ScalarParameter() {defaultValue=-9, parameterName=MoveVertAlongNormalsAmt}
  n68 = VertexNormalWS()
  n69 = Multiply(A=LinearInterpolate(A=Multiply(A=@n67, B=@n68), B=Constant(), Alpha=@n57), B=@n62.B)
  n70 = SetMaterialAttributes!(
   MaterialAttributes
   Base Color=NamedRerouteDeclaration(None=StaticSwitchParameter(True=@n29, False=@n30) {defaultValue=true, parameterName=UseNerdBackpack}) {name=basecolor, nodeColor=<rgba 0.9688 0 1 1>}
   Metallic=NamedRerouteDeclaration(None=Multiply(A=@n31.G, B=@n32)) {name=MEtallic, nodeColor=<rgba 1 0 0.6094 1>}
   Specular=NamedRerouteDeclaration(None=Multiply(A=@n31.R, B=@n33)) {name=Specular, nodeColor=<rgba 0 1 0.8828 1>}
   Roughness=NamedRerouteDeclaration(None=Multiply(A=@n31.B, B=@n34)) {name=Roughness, nodeColor=<rgba 1 0 0.1641 1>}
   Emissive Color=NamedRerouteDeclaration(None=Add(A=@n63, B=@n65)) {name=Emissive, nodeColor=<rgba 0 1 0.8594 1>}
   Normal=NamedRerouteDeclaration(None=@n66.RGB) {name=Normal, nodeColor=<rgba 0 1 0.7656 1>}
   World Position Offset=NamedRerouteDeclaration(None=StaticSwitchParameter(True=@n30, False=@n69) {defaultValue=true, parameterName=UseNerdBackpack}) {name=WPO, nodeColor=<rgba 0.7109 1 0 1>}
  ) {attributeSetTypes [
    #1
    #2
    #3
    #4
    #5
    #6
    #7
   ]}
  MaterialFunctionCall(In=@n70) {materialFunction=&MF_TechArtOverrides_accessories}
 ]
}
